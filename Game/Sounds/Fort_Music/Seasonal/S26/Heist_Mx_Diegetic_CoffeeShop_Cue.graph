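# Sound cue for diegetic coffee shop music in the Season 26 Heist music set, reusing an older lo-fi track.
# A single looping wave player feeds a full-level send to the S24 Heist coffee shop POI audio bus.
# The base submix is disabled, so the cue acts as a bus feed rather than direct output.
# Play-when-silent virtualization and a very large max distance keep the loop running.
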SoundCue /Game/Sounds/Fort_Music/Seasonal/S26/Heist_Mx_Diegetic_CoffeeShop_Cue
root=WavePlayer_0
nodes [{id=WavePlayer_0, properties {bLooping=true, soundWaveAssetPtr=&/s21_TikTok_LoFi_Loop}}]
properties {
 bEnableBaseSubmix=false
 maxDistance=2097152
 virtualizationMode=PlayWhenSilent
 volumeMultiplier=1
 busSends [{
   audioBus=&/S24_Heist_AB_CoffeeShop_POI
   maxSendDistance=2000
   minSendDistance=500
   minSendLevel=1
   sendLevel=1
   soundSourceBus=null
  }]
}
