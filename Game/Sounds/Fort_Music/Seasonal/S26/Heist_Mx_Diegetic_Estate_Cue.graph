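# Sound cue for the diegetic estate music of the Season 26 heist content in Fortnite.
# A single looping wave player plays the futzed lair music wave.
# The base submix is disabled, so the sound is routed through a distance-based send to the Estate POI audio bus, presumably for re-emission by sources on that bus.
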SoundCue /Game/Sounds/Fort_Music/Seasonal/S26/Heist_Mx_Diegetic_Estate_Cue
root=WavePlayer_2
nodes [{id=WavePlayer_2, properties {bLooping=true, soundWaveAssetPtr=&/s26_Lair_Diegetic_Music_futz_Pruned}}]
properties {
 bEnableBaseSubmix=false
 maxDistance=2097152
 volumeMultiplier=1
 busSends [{
   audioBus=&/S24_Heist_AB_Estate_POI
   maxSendDistance=2000
   minSendDistance=500
   minSendLevel=1
   sendLevel=1
   soundSourceBus=null
  }]
}
